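; Parent projectile for the Balloon Grenade, a shockwave-style grenade that launches what it hits and adds balloons to players and vehicles.
; Stopping or entering water starts a beep and delayed explosion; bouncing sets a short detonation timer.
; The blast deduplicates hits, then launches pawns, vehicles and physics objects via a shared velocity function with minimum lift.
; Player launches respect team and fall-damage settings, can damage structures along the path, and send quest events.
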(event EventBeginPlay
  ;[1 1 1] EVENT--> BeginPlay
  (Networking|CanExecuteCosmeticEvents
    (:True
      (Audio|Components|Audio|SetSound (Variables|B_Prj_Grenade|GetAudio_GrenadeFuse) (Variables|Sounds|GetCue_GrenadeFuseSound)))))

(event OnComponentBeginOverlap(CapsuleComponent)
  (params (OverlappedComponent :type Primitive%20Component%20Object%20Reference) (OtherActor :type Actor%20Object%20Reference) (OtherComp :type Primitive%20Component%20Object%20Reference) (OtherBodyIndex :type Integer) (bFromSweep :type Boolean) (SweepResult :type Hit%20Result%20Structure%20%28by%20ref%29))
  (bind (_bblockinghit _binitialoverlap _time _distance _location _impactpoint _normal _impactnormal _physmat _hitactor) (Collision|BreakHitResult SweepResult))
  ;[1 1 1] EVENT--> BeginOverlap (Stop Movement)
  (bind _aspawn (Utilities|Casting|CastToPawn _hitactor))
  (if (Utilities|NotEqual(Object) _aspawn (Game|GetInstigator))
    (Components|Movement|StopMovementImmediately (Variables|Movement|GetProjectileMovementComponent))))

(event EventOnTouched
  (params (OtherActor :type Actor%20Object%20Reference) (OtherComp :type Primitive%20Component%20Object%20Reference) (HitResult :type Hit%20Result%20Structure%20%28by%20ref%29) (bIsOverlap :type Boolean))
  (bind (_bblockinghit _binitialoverlap _time _distance _location _impactpoint _normal) (Collision|BreakHitResult HitResult))
  (bind (_bblockinghit_2 _binitialoverlap_2 _time_2 _distance_2 _location_2 _impactpoint_2 _normal_2 _impactnormal _physmat _hitactor _hitcomponent) (Collision|BreakHitResult HitResult))
  ;[1 1 1] If we touch water do nothing (bounce)
  (Utilities|Casting|CastToFortWaterBodyActor OtherActor
    (:CastFailed
      (if (not (Components|ComponentHasTag _hitcomponent "NoStick"))
        (Variables|Projectile|SetbStopSimulatingOnHit true)
        (Variables|BPrjGrenade|SetHitLocation _location)
        (Variables|BPrjGrenade|SetHitNormal _normal)))))

(event OnProjectileBounce(ProjectileMovementComponent)
  (params (ImpactResult :type Hit%20Result%20Structure%20%28by%20ref%29) (ImpactVelocity :type Vector%20%28by%20ref%29 :default "0, 0, 0"))
  ;[1 1 1] Bounce detonation timer
  (Utilities|Time|SetTimerbyEvent (EventDispatchers|CreateEvent :!event DoExplosion) 0.4))

(event Custom|Stop_Rotation
  ;[1 1 1] Stop Rotating Movement
  (Components|Movement|SetUpdatedComponent (Variables|Default|GetRotatingMovement) 0))

(event EventOnStop
  (params (Hit_bBlockingHit :type Boolean) (Hit_bInitialOverlap :type Boolean) (Hit_Time :type Float%20%28single-precision%29) (Hit_Distance :type Float%20%28single-precision%29) (Hit_Location :type Vector :default "0, 0, 0") (Hit_ImpactPoint :type Vector :default "0, 0, 0") (Hit_Normal :type Vector :default "0, 0, 0") (Hit_ImpactNormal :type Vector :default "0, 0, 0") (Hit_PhysMat :type Physical%20Material%20Object%20Reference) (Hit_HitActor :type Actor%20Object%20Reference) (Hit_HitComponent :type Primitive%20Component%20Object%20Reference) (Hit_HitBoneName :type Name :default "None") (Hit_BoneName :type Name :default "None") (Hit_HitItem :type Integer) (Hit_ElementIndex :type Integer) (Hit_FaceIndex :type Integer) (Hit_TraceStart :type Vector :default "0, 0, 0") (Hit_TraceEnd :type Vector :default "0, 0, 0"))
  (bind _returnvalue (Transformation|GetActorLocation))
  (bind _beeptimer (Variables|Default|GetBeepTimer))
  ;[1 1 1] EVENT--> OnStop
  (Variables|BPrjGrenade|SetHitLocation Hit_Location)
  (Variables|BPrjGrenade|SetHitNormal Hit_Normal)
  (Utilities|FlowControl|Sequence
    (:then_0
      (bind _returnvalue_2 (GameplayTags|HasAnyMatchingGameplayTags(Message) Hit_HitActor (Variables|BPrjGrenade|GetAttachableTags)))
      (if _returnvalue_2
        (Transformation|AttachActorToActor Hit_HitActor "None" "KeepWorld" "KeepWorld" "KeepWorld")
        (else
          _))
      (CallFunction|StopRotation)
      (Utilities|FlowControl|DoOnce false
        (:Completed
          (if (not (Water|IsinWater (Variables|Water|GetWaterInteractionComponent)))
            (Audio|PlaySoundatLocation (Variables|Sounds|GetCue_StopSound) _returnvalue "0, 0, 0" 0.6) ; Play stop audio
          )
          (Audio|Components|Audio|Play _beeptimer 0.0)
          (Audio|Components|Audio|FadeOut (Variables|Default|GetAudio_InAirLoop) 0.1 0.0 "Linear" :!class /Script/Engine.AudioComponent)
          (SoundIndicator|ShowSoundIndicator (Variables|Default|GetFortSoundIndicator) _returnvalue (Variables|/Script/Engine.AudioComponent|GetSound _beeptimer) 1.0 0)
          (Utilities|FlowControl|Delay (ScalableFloat|GetValueatLevel (Variables|BPrjGrenade|GetExplodeDelay)))
          (Utilities|FlowControl|Sequence
            (:then_0
              (if (Variables|Default|GetHitWater)
                (Variables|BPrjGrenade|SetHitLocation (Transformation|GetActorLocation)) ; Update explosion location to current location\nafter drifting in water
              ))
            (:then_1
              (Delivery|DoExplosion))))))
    (:then_1
      ;;[1 1 1] Dust Impact
      (Effects|Components|ParticleSystem|SpawnFXAtLocation (Variables|Default|GetFXSystem%20Asset) (Variables|BPrjGrenade|GetHitLocation) (Math|Conversions|RotationFromXVector (Variables|BPrjGrenade|GetHitNormal))))))

(event OnEnterWaterEvent(WaterInteractionComponent)
  (params (WaterBody :type Fort%20Water%20Body%20Actor%20Object%20Reference) (WaterInteractionComponent :type Fort%20Water%20Interaction%20Component%20Object%20Reference) (bIsFirstBody :type Boolean))
  (bind (_waterplanelocation _waterplanenormal _watersurfaceposition _waterdepth _waterbodyidx) (FX|GetLastWaterSurfaceInfo WaterInteractionComponent :!class /Script/FortniteGame.FortWaterInteractionComponent))
  (bind _returnvalue (Transformation|GetActorLocation))
  (bind _audio_grenadefuse (Variables|B_Prj_Grenade|GetAudio_GrenadeFuse))
  (bind _beeptimer (Variables|Default|GetBeepTimer))
  ;[1 1 1] EVENT --> OnEnterWater
  (Variables|Default|SetHitWater true)
  (Variables|BPrjGrenade|SetHitLocation (Transformation|GetActorLocation))
  (Variables|BPrjGrenade|SetHitNormal "0, 0,1.000000")
  (bind _returnvalue_2 (Effects|SpawnEmitteratLocation "/Game/Weapons/FORT_Generic_Effects/Impacts/Water/P_Pistol_Impact_Water_Apollo.P_Pistol_Impact_Water_Apollo" _watersurfaceposition (Math|Rotator|MakeRotfromX _waterplanenormal)))
  (Effects|Components|ParticleSystem|SetFloatParameter _returnvalue_2 "WaterDepth" _waterdepth)
  (Effects|Components|ParticleSystem|SetFloatParameter _returnvalue_2 "WaterBodyIndex" (Math|Conversions|ToFloat(Integer) _waterbodyidx))
  (Audio|Components|Audio|SetSound _audio_grenadefuse "/Game/Sounds/Fort_GamePlay_Sounds/Explosives/Grenade_Projectile_Bounce_Water_Splash_Cue.Grenade_Projectile_Bounce_Water_Splash_Cue")
  (Audio|Components|Audio|Play _audio_grenadefuse 0.0)
  ;[1 1 1] EVENT--> OnStop
  (Utilities|FlowControl|DoOnce false
    (:Completed
      (if (not (Water|IsinWater (Variables|Water|GetWaterInteractionComponent)))
        (Audio|PlaySoundatLocation (Variables|Sounds|GetCue_StopSound) _returnvalue "0, 0, 0" 0.6) ; Play stop audio
      )
      (Audio|Components|Audio|Play _beeptimer 0.0)
      (Audio|Components|Audio|FadeOut (Variables|Default|GetAudio_InAirLoop) 0.1 0.0 "Linear" :!class /Script/Engine.AudioComponent)
      (SoundIndicator|ShowSoundIndicator (Variables|Default|GetFortSoundIndicator) _returnvalue (Variables|/Script/Engine.AudioComponent|GetSound _beeptimer) 1.0 0)
      (Utilities|FlowControl|Delay (ScalableFloat|GetValueatLevel (Variables|BPrjGrenade|GetExplodeDelay)))
      (Utilities|FlowControl|Sequence
        (:then_0
          (if (Variables|Default|GetHitWater)
            (Variables|BPrjGrenade|SetHitLocation (Transformation|GetActorLocation)) ; Update explosion location to current location\nafter drifting in water
          ))
        (:then_1
          (Delivery|DoExplosion))))))

(event EventOnExploded
  (params (HitActors :type Array%20of%20Actor%20Object%20References) (HitResults :type Array%20of%20Hit%20Result%20Structures))
  (bind _returnvalue (Transformation|GetActorLocation))
  (bind _returnvalue_2 (Transformation|GetActorLocation))
  (bind _cue_closesound (Variables|Sounds|GetCue_CloseSound))
  (bind _target (Variables|BPrjGrenade|GetTarget))
  (bind _instigator (Variables|Actor|GetInstigator))
  (bind _target_2 (Variables|BPrjGrenade|GetTarget))
  (bind _target_3 (Variables|BPrjGrenade|GetTarget))
  (bind _target_4 (Variables|BPrjGrenade|GetTarget))
  (bind _mesh (Variables|/Script/Engine.Character|GetMesh _target_3))
  (bind _cue_distantsound (Variables|Sounds|GetCue_DistantSound))
  ;[1 1 1] EVENT--> OnExploded
  (bind _targets (CallFunction|RemoveDuplicateActors HitActors))
  (Audio|Components|Audio|FadeOut (Variables|Default|GetBeepTimer) 1.0 0.0 "Linear" :!class /Script/Engine.AudioComponent)
  (Utilities|FlowControl|Sequence
    (:then_0
      (for _array_element _targets
        (bind _asfort_pawn (Utilities|Casting|CastToFortPawn _array_element)
          (:then
            (bind _output_get (Variables|BPrjGrenade|SetTarget _asfort_pawn))
            (if (GameplayTags|HasMatchingGameplayTag _output_get "(TagName=\"Gameplay.Status.KnockbackImmunity\")")
              _
              (elif (Team|OnSameTeam (Variables|Actor|GetInstigator) (Variables|BPrjGrenade|GetTarget))
                ;;[1 0.212 0 0.5] Check if the Team Settings Device has overriden the ability to impulse other teammates. (Still impulse the instigating player regardless)
                (if (ScalableFloat|GetValueasBoolwithContextParam (Variables|/Script/Engine.Pawn|GetPlayerState _instigator) (Variables|Default|GetScalableFloat_AllowImpulsingTeammates))
                  _
                  (elif (== (Variables|BPrjGrenade|GetTarget) _instigator)
                    _))
                (if (> (ScalableFloat|GetValueatLevel (Variables|BPrjGrenade|GetAllPlayersTakeFallDamage)) 0.0)
                  _
                  (elif (> (ScalableFloat|GetValueatLevel (Variables|BPrjGrenade|GetEnemiesTakeFallDamage)) 0.0)
                    (if (Team|OnSameTeam (Variables|Actor|GetInstigator) _target_4)
                      (Variables|/Script/FortniteGame.FortPawn|SetbIgnoreNextFallingDamage true _target_4)
                      (Environment|ApplyGlobalEnvironmentGameplayEffecttoActor _target_4 "/Game/Athena/Items/Consumables/ShockwaveGrenade/GE_Athena_ShockGrenade_ImmunityGrant.GE_Athena_ShockGrenade_ImmunityGrant_C" 0 "(GameplayTags=)")
                      (else
                        _))
                    (bind _asfort_player_pawn (Utilities|Casting|CastToFortPlayerPawn (Variables|BPrjGrenade|GetTarget))
                      (:then
                        (if (not (Vehicle|IsinVehicle _asfort_player_pawn))
                          ;[1 1 1] Play trail FX on player, destroy structures in path
                          (Utilities|FlowControl|Sequence
                            (:then_0
                              (CallFunction|HandlePawnDetachRC)
                              (Components|Movement|StopMovementImmediately (Variables|/Script/Engine.Character|GetCharacterMovement _target_2))
                              (if (> (ScalableFloat|GetValueatLevel (Variables|BPrjGrenade|GetShouldDestroy)) 0.0)
                                (bind _outhits (Collision|MultiCapsuleTraceByChannel (Transformation|GetActorLocation _target_2) (+ (Transformation|GetActorLocation _target_2) (* (Math|Vector|GetUnitDirection(Vector) (Variables|BPrjGrenade|GetHitLocation) (Transformation|GetActorLocation _target_2)) (ScalableFloat|GetValueatLevel (Variables|BPrjGrenade|GetDestroyDistance)))) 8e+01 1.5e+02 "TraceTypeQuery11"))
                                (for _array_element_2 _outhits
                                  (if (Math|Boolean|ANDBoolean (Variables|/Script/Engine.Actor|GetbCanBeDamaged (.HitActor (Collision|BreakHitResult _array_element_2))) (Utilities|IsA (.HitActor (Collision|BreakHitResult _array_element_2)) "/Script/FortniteGame.BuildingActor") (not (Utilities|IsA (.HitActor (Collision|BreakHitResult _array_element_2)) "/Script/FortniteGame.BuildingTrap")))
                                    (Utilities|IsValid (Ability|GetAbilitySystemComponent (.HitActor (Collision|BreakHitResult _array_element_2)))
                                      (:Is%20Valid
                                        (if (not (GameplayTags|HasAnyMatchingGameplayTags (Ability|GetAbilitySystemComponent (.HitActor (Collision|BreakHitResult _array_element_2))) (GameplayTags|MakeLiteralGameplayTagContainer "(GameplayTags=((TagName=\"Deimos.Spawner\"),(TagName=\"TidalCrane.BugSpawner\"),(TagName=\"NPC.ProtoSwarm.Boss\"),(TagName=\"Granted.Athena.Item.ShockGrenade.ImmuneToPlayerImpulseDestruction\")))")))
                                          (GameplayEffects|ApplyGameplayEffectToTarget :self (Ability|GetAbilitySystemComponent (Variables|BPrjGrenade|GetTarget)) :GameplayEffectClass "/Game/Athena/Items/Consumables/ShockwaveGrenade/GE_ShockGrenade_Damage.GE_ShockGrenade_Damage_C" :Target (Ability|GetAbilitySystemComponent (.HitActor (Collision|BreakHitResult _array_element_2))) :Level 0.0))))))
                                (else
                                  _))
                              ;[1 1 1] Launch Player
                              (Utilities|FlowControl|Sequence
                                (:then_0
                                  (Creative|GameplayEffects|SimpleApplyGameplayEffectToActor "/Game/Athena/Items/Consumables/Balloons/GE_Athena_Balloons_Consumable_Passive.GE_Athena_Balloons_Consumable_Passive_C" _target)
                                  (Ability|SendGameplayEventtoActor _target "(TagName=\"Event.Athena.Balloons.Inflate\")"))
                                (:then_1
                                  (Utilities|FlowControl|DelayUntilNextTick)
                                  (Networking|SwitchHasAuthority
                                    (:Authority
                                      (Environment|ApplyGlobalEnvironmentGameplayEffecttoActor _target (Variables|Default|GetGE_KnockbackStatus) :!pinned true) ; Apply Knockback Status
                                      (Character|LaunchCharacter _target (CallFunction|CalculateLaunchVel (Variables|BPrjGrenade|GetTarget)) true true)
                                      (bind _asfort_player_pawn_athena (Utilities|Casting|CastToFortPlayerPawnAthena _target))
                                      (Class|FortPlayerPawnAthena|SetFallInstigator _asfort_player_pawn_athena (Game|GetInstigatorController) (Variables|Tags|GetDefaultTags))
                                      (CallFunction|QuestSendShockwaveTag))))))
                            (:then_1
                              (Effects|SpawnFXSystemAttached :FXSystemAsset (Variables|Default|GetNS_FXSystem%20Asset_trail) :AttachToComponent _mesh :AttachPointName "root" :Scale "1.000000,1.000000,1.000000" :LocationType "SnapToTarget" :bAutoDestroy true :PoolingMethod "None" :bAutoActivate true)
                              (Audio|SpawnSoundAttached "/Game/Sounds/Fort_GamePlay_Sounds/Explosives/ShockwaveGrenade/SC_Grenade_Shockwave_Launch_Impulse.SC_Grenade_Shockwave_Launch_Impulse" _mesh)))))
                      (:CastFailed
                        ;;[1 1 1] Was the actor hit Irwin? Send shockwave quest update event
                        (if (GameplayTags|HasMatchingGameplayTag (Variables|BPrjGrenade|GetTarget) "(TagName=\"Athena.Faction.Wildlife\")")
                          (CallFunction|QuestSendShockwaveTag))))
                    (else
                      _)))
                (else
                  _))))
          (:CastFailed
            (Default|TryAttachBalloontoVehicle _array_element (Variables|BPrjGrenade|GetHitLocation))
            (Utilities|FlowControl|DelayUntilNextTick)
            (CallFunction|LaunchNonPawnActors _array_element))))
      ;;[1 0.8789 0.2904] FortAIPawn
      (for _array_element_3 _targets
        (bind _asfort_aipawn (Utilities|Casting|CastToFortAIPawn _array_element_3))
        (if (not (GameplayTags|HasMatchingGameplayTag _asfort_aipawn "(TagName=\"Gameplay.Status.KnockbackImmunity\")"))
          (Variables|Default|SetFortAIPawn _asfort_aipawn)
          (CallFunction|LaunchFortAIPawn)))
      ;;[0.4913 1 0.5974] Launch Player Vehicle
      (for _array_element_4 _targets
        (bind _asfort_athena_vehicle (Utilities|Casting|CastToFortAthenaVehicle _array_element_4))
        (Variables|BPrjGrenade|SetTargetVehicle _asfort_athena_vehicle)
        (CallFunction|LaunchPlayerVehicle)))
    (:then_1
      ;[1 1 1] FX / Audio
      ;;[0.2951 0.5702 1] FX in the water do water explosion
      (Networking|CanExecuteCosmeticEvents
        (:True
          (if (Water|IsinWater (Variables|Water|GetWaterInteractionComponent))
            (bind _returnvalue_3 (Effects|Components|ParticleSystem|SpawnAltFXAtLocation "GenericAnimNotify" (Variables|BPrjGrenade|GetNS_WaterExplosion) (Variables|BPrjGrenade|GetP_WaterExplosion) (select (Variables|Projectile|GetbStopSimulatingOnHit) (Variables|BPrjGrenade|GetHitLocation) (Transformation|GetActorLocation)) (Math|Rotator|MakeRotfromZ (0.0 0.0 1.0))))
            (Development|PrintString (Utilities|GetDisplayName _returnvalue_3))
            (else
              (Effects|Components|ParticleSystem|SpawnAltFXAtLocation "GenericAnimNotify" (Variables|BPrjGrenade|GetNS_Explosion) (Variables|BPrjGrenade|GetP_Explosion) (select (Variables|Projectile|GetbStopSimulatingOnHit) (Variables|BPrjGrenade|GetHitLocation) (Transformation|GetActorLocation)) (Math|Conversions|RotationFromXVector (Variables|BPrjGrenade|GetHitNormal)))))
          (GameplayCue|ExecuteGameplayCue%28Burst%29Local self (Variables|Default|GetFeedbackCue) (Ability|GameplayCue|MakeGameplayCueParameters 0.0)) ; Cam shake + controller vibration
          (if (Pawn|IsLocallyControlled (Game|GetInstigator))
            (Audio|PlaySoundatLocation _cue_closesound _returnvalue_2 "0, 0, 0" 0.6) ; Play Close Explosion
            (Fortnite|ShowSoundIndicatorLocal :Location _returnvalue_2 :MaxAudibleDistance (Variables|/Script/Engine.SoundBase|GetMaxDistance _cue_closesound) :IndicatorType "Gunshot" :Instigator self :Tint "(R=0.000000,G=0.000000,B=0.000000,A=0.000000)" :AttenuationMaxRangeMultiplier 1.0 :AttenuationIntensityMultiplier 1.0)
            (else
              (Audio|PlaySoundatLocation _cue_distantsound _returnvalue "0, 0, 0" 0.4) ; Play Distant Explosion
              (Fortnite|ShowSoundIndicatorLocal :Location _returnvalue :MaxAudibleDistance (Variables|/Script/Engine.SoundBase|GetMaxDistance _cue_distantsound) :IndicatorType "Gunshot" :Instigator self :Tint "(R=0.000000,G=0.000000,B=0.000000,A=0.000000)" :AttenuationMaxRangeMultiplier 1.0 :AttenuationIntensityMultiplier 1.0)))
          (Audio|Components|Audio|FadeOut (Variables|B_Prj_Grenade|GetAudio_GrenadeFuse) 0.2 0.0 "Linear" :!class /Script/Engine.AudioComponent) ; Fade Out Fuse Sound
        )))))

(event Custom|LaunchFortAIPawn
  (bind _fortaipawn (Variables|Default|GetFortAIPawn))
  ;[0.967 1 0.5015] Launch FortAIPawn
  (Utilities|FlowControl|Sequence
    (:then_0
      (Utilities|IsValid _fortaipawn
        (:Is%20Valid
          (Components|Movement|StopMovementImmediately (Variables|/Script/Engine.Character|GetCharacterMovement _fortaipawn))
          (Environment|ApplyGlobalEnvironmentGameplayEffecttoActor _fortaipawn (Variables|Default|GetGE_KnockbackStatus) :!pinned true) ; Apply Knockback Status
          (Character|LaunchCharacter _fortaipawn (CallFunction|CalculateLaunchVel (Variables|Default|GetFortAIPawn)) true true))))
    (:then_1
      (Effects|SpawnFXSystemAttached :FXSystemAsset (Variables|Default|GetNS_FXSystem%20Asset_trail) :AttachToComponent (Variables|/Script/Engine.Character|GetMesh (Variables|Default|GetFortAIPawn)) :AttachPointName "root" :Scale "1.000000,1.000000,1.000000" :LocationType "SnapToTarget" :bAutoDestroy true :PoolingMethod "None" :bAutoActivate true))))

(event Custom|LaunchNonPawnActors
  (params (Actor :type Actor%20Object%20Reference))
  ;[1 1 1] FX / Audio
  (bind _asfort_physics_pawn_object (Utilities|Casting|CastToFortPhysicsPawnObject Actor)
    (:then
      (Utilities|IsValid _asfort_physics_pawn_object
        (:Is%20Valid
          (bind (_initialparameters_initialvelocity _initialparameters_overridemassinkg _initialparameters_overrideangulardampening _initialparameters_overridelineardampening _initialparameters_playerforcemultiplier) (Utilities|Struct|BreakPhysicsPawnObjectInitialParameters (Variables|/Script/FortniteGame.FortPhysicsPawnObject|GetInitialParameters_InitialVelocity _asfort_physics_pawn_object)))
          (PhysicsObject|ApplyImpulse _asfort_physics_pawn_object self (* (CallFunction|CalculateLaunchVel Actor) (* (/ _initialparameters_playerforcemultiplier (Math|Float|Max(Float) (Variables|Default|GetPhysicsPawnObjectForceModifier) 1.0)) (ScalableFloat|GetValueatLevel (Variables|Default|GetRow_PhysicsPawnObjectMult))))))))
    (:CastFailed
      (CallFunction|LaunchBuldingActors Actor))))

(event Custom|LaunchPlayerVehicle
  (bind _targetvehicle (Variables|BPrjGrenade|GetTargetVehicle))
  ;[0.5431 1 0.6714] Launch Player Vehicle
  (Vehicle|WakeUp _targetvehicle)
  (Audio|BroadcastAbilityImpactatLocation 0 _targetvehicle (* (CallFunction|CalculateLaunchVel _targetvehicle) (ScalableFloat|GetValueatLevel (Variables|Default|GetRow_VehicleImpulseMult))) 777.0)
  (for _array_element (Vehicle|GetAllPassengers (Variables|BPrjGrenade|GetTargetVehicle))
    (Variables|BPrjGrenade|SetTarget _array_element)
    (CallFunction|QuestSendShockwaveTag))
  (Effects|SpawnEmitterAttached :EmitterTemplate "/Game/Athena/Items/Consumables/KnockGrenade/P_Athena_KnockGrenade.P_Athena_KnockGrenade" :AttachPointName "None" :Scale "1.000000,1.000000,1.000000" :LocationType "SnapToTarget" :bAutoDestroy true :PoolingMethod "None" :bAutoActivate true))

(event Custom|LaunchBuldingActors
  (params (Actor :type Actor%20Object%20Reference))
  ;[1 1 1] If the actor has the building actor tag and also a physics component launch it.
  (bind _returnvalue (GameplayTags|HasMatchingGameplayTag(Message) Actor (Variables|Default|GetTag_BuildingPhysocs)))
  (if _returnvalue
    (Utilities|IsValid (Actor|GetComponentbyClass Actor "/Script/FortniteGame.FortPhysicsObjectComponent")
      (:Is%20Valid
        (Utilities|Casting|CastToPrimitiveComponent (Variables|/Script/Engine.Actor|GetRootComponent Actor))
        (FortPhysicsSimulationLibrary|AddPhysicsImpulseatLocation Actor (* (CallFunction|CalculateLaunchVel Actor) (ScalableFloat|GetValueatLevel (Variables|Default|GetRow_BuildingImpulseMult))) (Variables|BPrjGrenade|GetHitLocation) true "None" false 0 "(MaxSimSizeAllowed=Large,bRejectEmbeddedInLandscape=False)")))))

(event Custom|QuestSendShockwaveTag
  (bind _returnvalue (Game|GetInstigatorController))
  ;[1 1 1] EVENT--> Send a quest stat event (sourcetag: Athena.Quests.ShockwaveGrenade), called when a Shockwave Grenade knocks back an actor
  (if (Team|OnSameTeam (Variables|BPrjGrenade|GetTarget) (Variables|Actor|GetInstigator))
    (if (Utilities|Equal(Object) (Variables|BPrjGrenade|GetTarget) (Variables|Actor|GetInstigator))
      (Utilities|Casting|CastToFortPlayerController _returnvalue)
      (CustomVerb|BroadcastCustomVerbMessage (CustomVerb|MakeCustomVerbMessage (Variables|Actor|GetInstigator) "(TagName=\"CustomVerb.ShockwaveSelf\")")))
    (else
      (bind _asfort_player_controller (Utilities|Casting|CastToFortPlayerController _returnvalue))
      (CustomVerb|BroadcastCustomVerbMessage (CustomVerb|MakeCustomVerbMessage (Variables|Actor|GetInstigator) "(TagName=\"CustomVerb.ShockwaveTarget\")"))
      ;[1 1 1] Send Gameplay event so quest abilities can utilize the target actor for post knockback handling
      (Ability|SendGameplayEventtoActor _asfort_player_controller "(TagName=\"CustomVerb.ShockwaveTarget\")" (Utilities|Struct|MakeGameplayEventData "(TagName=\"\")" (Variables|Actor|GetInstigator) (Variables|BPrjGrenade|GetTarget))))))

(fn CalculateLaunchVel
  (params (Target :type Actor%20Object%20Reference))
  ;[1 1 1] Calculate the launch velocity for the actor
  (return (Math|Vector|MakeVector (.x (* (Math|Vector|Normalize (- (Transformation|GetActorLocation Target) (CallFunction|GetLaunchOriginPoint))) (ScalableFloat|GetValueatLevel (Variables|BPrjGrenade|GetLaunchVelocity)))) (.y (* (Math|Vector|Normalize (- (Transformation|GetActorLocation Target) (CallFunction|GetLaunchOriginPoint))) (ScalableFloat|GetValueatLevel (Variables|BPrjGrenade|GetLaunchVelocity)))) (Math|Float|Max(Float) (+ (.z (* (Math|Vector|Normalize (- (Transformation|GetActorLocation Target) (CallFunction|GetLaunchOriginPoint))) (ScalableFloat|GetValueatLevel (Variables|BPrjGrenade|GetLaunchVelocity)))) 2.5e+02) 5e+02))))

(fn GetLaunchOriginPoint
  (return (Variables|BPrjGrenade|GetHitLocation)))

(fn HandlePawnDetachRC
  (bind _asfort_player_pawn (Utilities|Casting|CastToFortPlayerPawn (Variables|BPrjGrenade|GetTarget)))
  (RemoteControlPawn|ForceDetachfromRemoteControlledPawn _asfort_player_pawn))

(fn RemoveDuplicateActors
  (params (Actors :type Array%20of%20Actor%20Object%20References))
  (Utilities|Set|AddItems (LocalVariables|GetSetActors) Actors)
  (bind _result (Utilities|Set|ToArray (LocalVariables|GetSetActors)))
  (return _result))

(fn TryAttachBalloonToVehicle
  (params (HitActor :type Actor%20Object%20Reference) (HitLocation :type Vector))
  (if (ScalableFloat|GetValueasBool (Variables|Default|GetHF_AttachToVehicles?))
    (bind _asfort_athena_vehicle (Utilities|Casting|CastToFortAthenaVehicle HitActor))
    ;[1 1 1] Don't attach a balloon the vehicle already has one
    (Utilities|IsValid (Actor|GetComponentbyClass _asfort_athena_vehicle "/Script/FortniteGame.FortPhysicsConstantForceComponent")
      (:Is%20Valid
        (if (Components|Activation|IsActive (Actor|GetComponentbyClass _asfort_athena_vehicle "/Script/FortniteGame.FortPhysicsConstantForceComponent"))
          _
          (else
            _)))
      (:Is%20Not%20Valid
        _))
    (Game|SpawnActorfromClass "/Game/Athena/Items/Consumables/ShockwaveGrenade/BGA_PhysicsBalloons_Async_Standard.BGA_PhysicsBalloons_Async_Standard_C" (Math|Transform|MakeTransform HitLocation) "Undefined" "MultiplyWithRoot" 0 _asfort_athena_vehicle "None" 0 0 0 0 0)))

(fn ConstructionScript
  (Networking|CanExecuteCosmeticEvents
    (:True
      (Components|Movement|SetUpdatedComponent (Variables|Default|GetRotatingMovement) (Variables|Default|GetShockGrenadeMesh)))))
