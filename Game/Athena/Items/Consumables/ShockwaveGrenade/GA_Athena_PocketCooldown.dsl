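; GA_Athena_PocketCooldown is a gameplay ability filed under the Shockwave Grenade consumable that conditionally commits an ability cooldown based on the current storm phase.
; On activation it reads the current safe zone phase and compares it with the integer value of a storm phase scalable float variable.
; While the current phase is lower than that threshold it commits the cooldown, then ends the ability in either case.
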(event Ability|EventActivateAbility
  (bind _returnvalue (SafeZone|GetCurrentSafeZonePhase))
  (if (< _returnvalue (ScalableFloat|GetValueasInteger (Variables|Default|Getstormphase)))
    (Ability|CommitAbilityCooldown false))
  (Ability|EndAbility))
