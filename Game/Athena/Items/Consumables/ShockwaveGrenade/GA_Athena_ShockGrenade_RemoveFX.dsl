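; Shockwave grenade gameplay ability that keeps a looping FX cue on the launched player, then swaps it for a landing cue and ends.
; Activation sets up Walking and Custom movement-mode waits, a repeating vehicle-landed timer and a 15-second failsafe delay, whose completion also clears ignore-next-fall-damage.
; The timer finishes the effect when the player's vehicle touches ground, a fix for FX persisting after exiting vehicles (FORT-145705).
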(event Ability|EventActivateAbility
  (bind _playerpawn (Variables|Default|GetPlayerPawn))
  (bind _asfort_player_pawn (Utilities|Casting|CastToFortPlayerPawn (Ability|GetAvatarActorfromActorInfo))
    (:then
      (Variables|Default|SetPlayerPawn _asfort_player_pawn)
      (Ability|AddGameplayCueToOwner (Variables|Default|GetLooping%20GC))
      (Utilities|FlowControl|Sequence
        (:then_0
          (Ability|Tasks|WaitMovementModeChange "MOVE_Walking"
            (:then
              (bind _returnvalue (Utilities|Time|SetTimerbyEvent (EventDispatchers|CreateEvent :!event Custom|VehicleLandedCheck) 0.2 true))
              (Variables|Default|SetVehicleOnGroundTimer _returnvalue)
              (Utilities|FlowControl|RetriggerableDelay 15.0) ; Failsafe to ensure ability ends.
              (if (Variables|/Script/FortniteGame.FortPawn|GetbIgnoreNextFallingDamage _playerpawn)
                (Variables|/Script/FortniteGame.FortPawn|SetbIgnoreNextFallingDamage false _playerpawn)
                (else
                  _))
              (Ability|RemoveGameplayCueFromOwner (Variables|Default|GetLooping%20GC))
              (Ability|AddGameplayCueToOwner (Variables|Default|GetPlayer%20Landing%20GC))
              (Utilities|Time|ClearandInvalidateTimerbyHandle (Variables|Default|GetVehicleOnGroundTimer))
              (Ability|RemoveGrantedbyEffect)
              (Ability|EndAbility))
            (:OnChange
              _)))
        (:then_1
          (Ability|Tasks|WaitMovementModeChange "MOVE_Custom"
            (:OnChange
              _)))))
    (:CastFailed
      (Ability|EndAbility))))

(event Custom|VehicleLandedCheck
  (bind _playerpawn (Variables|Default|GetPlayerPawn))
  (bind _returnvalue (Vehicle|GetVehicleInterface _playerpawn))
  ;[1 1 1] So Bouncer FX stay on a player if they exit a vehicle before landing. Fix FORT-145705
  (Utilities|IsValid (Utilities|ToObject(Interface) _returnvalue)
    (:Is%20Valid
      (if (and (Class|FortVehicleInterface|IsTouchingGround _returnvalue) (Vehicle|IsinVehicle _playerpawn))
        (Utilities|Time|ClearandInvalidateTimerbyHandle (Variables|Default|GetVehicleOnGroundTimer))
        (Ability|RemoveGameplayCueFromOwner (Variables|Default|GetLooping%20GC))
        (Ability|AddGameplayCueToOwner (Variables|Default|GetPlayer%20Landing%20GC))
        (Utilities|Time|ClearandInvalidateTimerbyHandle (Variables|Default|GetVehicleOnGroundTimer))
        (Ability|RemoveGrantedbyEffect)
        (Ability|EndAbility)))))
