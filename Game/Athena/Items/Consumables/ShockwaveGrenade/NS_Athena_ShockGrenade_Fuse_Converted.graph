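# Niagara fuse effect for the Shockwave Grenade, apparently converted from Cascade, built from three emitters.
# Ribbon spawns per unit travelled, leaving an upward-accelerating trail that widens while fading from bright purple-blue.
# ElectricRing bursts and continuously spawns short-lived, randomly oriented meshes that grow, with curve-driven HDR color and material parameters for an electric look.
# Particle_Emitter bursts a purple-blue point light using curve-driven alpha.
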NiagaraSystem /Game/Athena/Items/Consumables/ShockwaveGrenade/NS_Athena_ShockGrenade_Fuse_Converted
dependencies {
 usedDataInterfaces [&NiagaraDataInterfaceCurve, &NiagaraDataInterfaceVectorCurve]
 usedDynamicInputs []
 usedModules [&SystemState]
}
emitters {
 ElectricRing {
  emitterUpdateScript [
   &EmitterState(Life Cycle Mode=::NewEnumerator1, Loop Duration=1, UseLoopDelay=true)
   &SpawnBurst_Instantaneous(Spawn Time=0.2, Use Spawn Probability=false, UseLoopCountLimit=false) {module=SpawnBurst_Instantaneous001}
   &SpawnBurst_Instantaneous(Use Spawn Probability=false, UseLoopCountLimit=false)
   &SpawnRate(SpawnRate=10, Use Spawn Probability=false)
  ]
  particleSpawnScript [
   /SetVariables(Particles.MeshOrientation=&MultiplyQuaternion(
     Quaternion A=&MakeQuatFromAxisAngle(
      Axis=&RandomRangeVector(Minimum=<0 0 0>, Maximum=<1 1 1>)
      AngleInDegrees=&RandomRangeFloat(Minimum=0, Maximum=360)
     )
     Quaternion B=$Engine.Owner.Rotation
    )) {isSetParameters=true}
   &InitializeParticle(
    Lifetime=&RandomRangeFloat(Minimum=0.2, Maximum=0.45)
    Mesh Scale=&RandomRangeVector(Minimum=<8 8 8>, Maximum=<10 10 25>)
    Mesh Scale Mode=::NewEnumerator4
    UsePositionOffset=false
    Write Material Random=true
   )
   &Color(Color=&MakeLinearColorFromVectorAndFloat("Vector (RGB)"=<5.5863 4.5716 7.5>, "Float (Alpha)"=&ClampFloat(Float=1)))
  ]
  particleUpdateScript [
   &ParticleState()
   &ScaleMeshSize(Scale Factor=&VectorFromFloat(Value=&FloatFromCurve(FloatCurve=DataInterface() {propertyValues {Curve {keys [{arriveTangent=1.2654, interpMode=RCIM_Cubic, time=0, value=0.7}, <at 1 2>]}}})))
   &DynamicMaterialParameters(
    Index 0 Param 2=&FloatFromCurve(FloatCurve=DataInterface() {propertyValues {Curve {keys [{arriveTangent=-0.1368, time=0.0404, value=-0.8632}, <at 1 -1>]}}})
    Index 0 Param 4=0.8
    Module.Index0Param0Write=false
    Module.Index0Param1Write=true
    Module.Index0Param2Write=false
    Module.Index0Param3Write=true
   ) {module=DynamicMaterialParameters001}
   &ScaleColor(
    Color Value To Scale=$Particles.Color
    Scale RGB=&VectorFromCurve(VectorCurve=DataInterface() {propertyValues {
       XCurve {keys [<at 0.5522 11.1725 RCIM_Cubic>,<at 0.9839 10.4772 RCIM_Cubic>]}
       YCurve {keys [<at 0.5522 9.1432 RCIM_Cubic>,<at 0.9839 2.3602 RCIM_Cubic>]}
       ZCurve {keys [<at 0.5522 15 RCIM_Cubic>,<at 0.9839 15 RCIM_Cubic>]}
      }})
    ScaleA=true
    ScaleRGB=true
   )
  ]
  rendererClasses [&NiagaraMeshRendererProperties]
  renderers [{rendererClass=&NiagaraMeshRendererProperties, rendererIndex=0}]
 }
 Particle_Emitter {
  emitterUpdateScript [
   &EmitterState(
    Inactive Response=::NewEnumerator1
    Life Cycle Mode=::NewEnumerator1
    Loop Behavior=::NewEnumerator1
    Loop Duration Mode=::NewEnumerator1
    UseLoopDelay=false
   )
   &SpawnBurst_Instantaneous(Use Spawn Probability=false, UseLoopCountLimit=false)
  ]
  particleSpawnScript [&InitializeParticle(Sprite Size=<1 1>, Sprite Size Mode=::NewEnumerator4, UsePositionOffset=false, Write Material Random=true)]
  particleUpdateScript [
   &ParticleState(Kill Particles When Lifetime Has Elapsed=false, Let Infinitely Lived Particles Die When Emitter Deactivates=true) {enabled=false}
   /SetVariables(Particles.LightExponent=1.5, Particles.LightRadius=80) {isSetParameters=true}
   &Color(Color=&MakeLinearColorFromVectorAndFloat(
     "Vector (RGB)"=<0.3592 0.2257 1>
     "Float (Alpha)"=&ClampFloat(
      Float=&FloatFromCurve(FloatCurve=DataInterface() {propertyValues {Curve {keys [<at 0 1>,<at 1 0>]}}}, Scale Curve=4)
      Max=4
     )
    ))
   &CascadeConversion_LightProperties(LightBrightnessScale=1, LightColorScale=<3 0 5>, LightRadiusScale=80) {enabled=false}
  ]
  rendererClasses [&NiagaraLightRendererProperties]
  renderers [{rendererClass=&NiagaraLightRendererProperties, rendererIndex=0}]
 }
 Ribbon {
  emitterUpdateScript [
   &EmitterState(Life Cycle Mode=::NewEnumerator1, Loop Duration=1, UseLoopDelay=true)
   &SpawnPerUnit(
    Module.Use Max Movement Threshold=true
    Module.Use Movement Tolerance=true
    Movement Tolerance=0.1
    Spawn Spacing:NiagaraFloat=&Divide_Float(A=1, B=&Divide_Float(A=1, B=50))
    Use Spawn Probability=false
   )
  ]
  particleSpawnScript [
   /SetVariables(Particles.Acceleration:Vector3f=<0 0 250>) {isSetParameters=true}
   &InitializeRibbon(Ribbon Width=&Multiply_Float(A=15, B=2), UsePositionOffset=false, Write Material Random=true)
  ]
  particleUpdateScript [
   /SetVariables(Particles.ScaledRibbonWidth:NiagaraFloat=&Multiply_Float(
     A=$Particles.RibbonWidth
     B=&FloatFromCurve(FloatCurve=DataInterface() {propertyValues {Curve {keys [
          {arriveTangent=9.5068, interpMode=RCIM_Cubic, time=-0.0056, value=0.0871}
          {arriveTangent=0.0733, leaveTangent=0.0733, time=0.9928, value=6.8584}
         ]}}})
    )) {isSetParameters=true}
   &ParticleState()
   &Color(Color=&MakeLinearColorFromVectorAndFloat(
     "Vector (RGB)"=&VectorFromCurve(VectorCurve=DataInterface() {propertyValues {
        XCurve {keys [
          <at 0.1712 4.4054>
          {arriveTangent=-13.8185, interpMode=RCIM_Cubic, leaveTangent=-13.8185, time=0.3943, value=0.5958}
          {arriveTangent=-1.7748, interpMode=RCIM_Cubic, leaveTangent=-1.7748, time=0.4653, value=0.342}
          <at 0.73 0 RCIM_Cubic>
         ]}
        YCurve {keys [
          <at 0.1712 1.1394>
          {arriveTangent=-3.423, interpMode=RCIM_Cubic, leaveTangent=-3.423, time=0.3943, value=0.6055}
          {arriveTangent=-1.8037, interpMode=RCIM_Cubic, leaveTangent=-1.8037, time=0.4653, value=0.1328}
          <at 0.73 0 RCIM_Cubic>
         ]}
        ZCurve {keys [
          <at 0.1712 12>
          {arriveTangent=-38.7269, interpMode=RCIM_Cubic, leaveTangent=-38.7269, time=0.3943, value=0.6173}
          {arriveTangent=-1.8389, interpMode=RCIM_Cubic, leaveTangent=-1.8389, time=0.4653, value=0.6121}
          <at 0.73 0 RCIM_Cubic>
         ]}
       }})
     "Float (Alpha)"=&ClampFloat(Float=&FloatFromCurve(FloatCurve=DataInterface() {propertyValues {Curve {keys [<at 0.1016 1>, {arriveTangent=-1.1131, interpMode=RCIM_Cubic, leaveTangent=-1.1131, time=0.4637, value=0.2405}, <at 1 0>]}}}))
    ))
   &AccelerationForce(Acceleration=$Particles.Acceleration)
   &SolveForcesAndVelocity(Clamp Velocity=false, Limit Acceleration=false)
   &DynamicMaterialParameters(
    Index 0 Param 1=&FloatFromCurve(FloatCurve=DataInterface() {propertyValues {Curve {keys [{arriveTangent=-2.8011, interpMode=RCIM_Cubic, time=0, value=1}, <at 1 0>]}}})
    Index 0 Param 2=1
    Index 0 Param 3=1
    Index 0 Param 4=1
    Module.Index0Param0Write=true
    Module.Index0Param1Write=true
    Module.Index0Param2Write=true
    Module.Index0Param3Write=true
   )
  ]
  rendererClasses [&NiagaraRibbonRendererProperties]
  renderers [{rendererClass=&NiagaraRibbonRendererProperties}]
 }
}
summary {
 systemName=NS_Athena_ShockGrenade_Fuse_Converted
 emitters [
  {emitterName=Ribbon, rendererClasses [&NiagaraRibbonRendererProperties]}
  {emitterName=Particle_Emitter, rendererClasses [&NiagaraLightRendererProperties]}
  {emitterName=ElectricRing, rendererClasses [&NiagaraMeshRendererProperties]}
 ]
 userVariables []
}
